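# Material function for the Cabbage Sugar backpack that animates its emissive screen with scanlines and a blink.
# A sine pulses a small SDF-box region of UV space, while two scrolling scanline layers come from a shared holographic function, the second changing tiling with camera distance via a sphere mask.
# These effects modulate emissive, add faint scanlines to dark screen areas clipped from lit color's blue channel, and are corrected by inverse eye adaptation.
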MaterialFunction /BRCosmetics/Accessories/FORT_Backpacks/Backpack_M_Med_CabbageSugar/Materials/MF_Backpack_CabbageSugar
graph {nodes [
  n1 = Multiply(A=FunctionInput() {inputName=Emissive}, B=NamedRerouteUsage())
  n2 = LinearInterpolate(
   A=@n1
   B=Multiply(A=@n1, B=NamedRerouteUsage())
   Alpha=ScalarParameter() {defaultValue=0.75, group=Cabbage Sugar - Backpack, parameterName=Screen Text - Scanline - Intensity, sortPriority=6}
  )
  n3 = ComponentMask(None=FunctionInput() {bUsePreviewValueAsDefault=true, inputName=Color Lit}) {b=true}
  n4 = Multiply(
   A=Multiply(A=NamedRerouteUsage(), B=Step(Y=@n3) {constX=0.05, desc=Clip color map to get dark screen areas as a mask.})
   B=ScalarParameter() {defaultValue=0.025, group=Cabbage Sugar - Backpack, parameterName=Black Screen - Scanline - Intensity, sortPriority=7}
  ) {constB=0.1}
  FunctionOutput(None=Multiply(A=Add(A=@n2, B=@n4), B=EyeAdaptationInverse().EyeAdaptationInverse)) {outputName=Emissive}
  n5 = TextureCoordinate()
  n6 = Constant3Vector() {desc=Position of Square Mask for Sine Animation, constant=<rgba 0.216 -0.412 0 0>}
  n7 = MaterialFunctionCall(X=@n6.R, Y=@n6.R) {materialFunction=&MakeFloat2}
  n8 = Constant3Vector() {desc=Size of Square Mask for Sine Animation, constant=<rgba 0.2 0.061 0 0>}
  n9 = Step(Y=MaterialFunctionCall!(
    UVs=Add(A=@n5, B=@n7.Result) {constB=-0.5}
    Size=MaterialFunctionCall(X=@n8.R, Y=@n8.R) {materialFunction=&MakeFloat2}.Result
    CornerRoundness
    KeepAspectRatio
   ) {materialFunction=/FortUILibrary/MaterialFunctions/Shapes/SDFs/MF_SDF_Box}.SDF) {constX=0}
  n10 = OneMinus(None=@n9)
  n11 = Time()
  n12 = ScalarParameter() {defaultValue=0.5, group=Cabbage Sugar - Backpack, parameterName=Blink Speed}
  n13 = Multiply(
   A=@n9
   B=MaterialFunctionCall!(Sine Phase=Multiply(A=@n11, B=@n12), Value 1, Value 2) {materialFunction=/Engine/Functions/Engine_MaterialFunctions02/Utility/Sine_Remapped}.Remapped Sine
  ) {constB=0.5}
  NamedRerouteDeclaration(None=Saturate(None=Max(A=@n10, B=@n13))) {name=Blinking, nodeColor=<rgba 0 1 0.2031 1>}
  n14 = ScalarParameter() {defaultValue=40, group=Cabbage Sugar - Backpack, parameterName=Scanline 02 - Tiling - Far, sortPriority=2}
  n15 = ScalarParameter() {defaultValue=200, group=Cabbage Sugar - Backpack, parameterName=Scanline 02 - Tiling - Close, sortPriority=3}
  n16 = SphereMask(
   A=CameraPositionWS()
   B=ObjectPositionWS()
   Radius=ScalarParameter() {defaultValue=150, group=Cabbage Sugar - Backpack, parameterName=Scanlines 01 - Distance To Object, sortPriority=4}
  ) {desc=Camera Based Distance Masking, hardnessPercent=0}
  NamedRerouteDeclaration(None=MaterialFunctionCall!(
    ScanLine_Tiling_1=ScalarParameter() {defaultValue=5, group=Cabbage Sugar - Backpack, parameterName=Scanline 01 - Tiling, sortPriority=0}
    Scanline_PannerSpeed_1=ScalarParameter() {defaultValue=0.1, group=Cabbage Sugar - Backpack, parameterName=Scanline 01 - Speed, sortPriority=1}
    ScanLine_Tiling=LinearInterpolate(A=@n14, B=@n15, Alpha=@n16)
    Scanline_PannerSpeed=ScalarParameter() {defaultValue=1.2, group=Cabbage Sugar - Backpack, parameterName=Scanline 02 - Speed, sortPriority=5}
    Scanline_Rotation
    UVs
   ) {materialFunction=/Game/Global/Materials/MaterialGym/Materials/HolographicTexture/MF_HolographicTexture_ScanLines}.Result) {name=Scanlines, nodeColor=<rgba 0.0036 0.0278 1 1>}
 ]}
