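# Physics asset for the CabbageSugar medium backpack (back bling) cosmetic mesh.
# It holds a single body on the pack bone, shaped as one rotated capsule that approximates the bag.
# With no constraints, it defines a simple collision and bounds volume rather than a jointed simulation.
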PhysicsAsset /BRCosmetics/Accessories/FORT_Backpacks/Backpack_M_Med_CabbageSugar/Meshes/M_MED_CabbageSugar_Pack_PhysicsAsset
bodies {pack [{shapeType=Capsule, length=15.2109, radius=13.9626, center=<-0.0431 -1.992 -3.2184>, rotation=<rot 0 -90.008 -89.4465>}]}
constraints []
